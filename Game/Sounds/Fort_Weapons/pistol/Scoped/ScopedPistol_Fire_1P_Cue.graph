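# First-person gunfire sound cue for the scoped pistol.
# A quality-level switch sends its first slot to a random pick of three close fire waves, while the other three slots share a branch with one fixed wave.
# Both branches add slight pitch and volume modulation, then pass through two nested mixers whose second inputs are empty.
# First-person weapon attenuation, sound class and gunfire concurrency settings govern playback.
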SoundCue /Game/Sounds/Fort_Weapons/pistol/Scoped/ScopedPistol_Fire_1P_Cue
root=QualityLevel_0
nodes [
 {id=Mixer_0, children [Mixer_1,null], properties {inputVolume [1,0.45]}}
 {id=Mixer_1, children [Modulator_2,null], properties {inputVolume [0.9,0.7]}}
 {id=Mixer_10, children [Modulator_13,null], properties {inputVolume [0.9,0.7]}}
 {id=Mixer_9, children [Mixer_10,null], properties {inputVolume [1,0.45]}}
 {id=Modulator_13, children [WavePlayer_11], properties {pitchMax=1.1, pitchMin=1, volumeMax=1.1, volumeMin=1}}
 {id=Modulator_2, children [Random_2], properties {pitchMax=1.1, pitchMin=1, volumeMax=1.1, volumeMin=1}}
 {id=QualityLevel_0, children [Mixer_0,Mixer_9,Mixer_9,Mixer_9]}
 {id=Random_2, children [WavePlayer_0,WavePlayer_1,WavePlayer_4], properties {weights [1,1,1]}}
 {id=WavePlayer_0, properties {soundWaveAssetPtr=&/ScopedPistol_Fire_Close_01}}
 {id=WavePlayer_1, properties {soundWaveAssetPtr=&/ScopedPistol_Fire_Close_02}}
 {id=WavePlayer_11, properties {soundWaveAssetPtr=&ScopedPistol_Fire_Close_01}}
 {id=WavePlayer_4, properties {soundWaveAssetPtr=&/ScopedPistol_Fire_Close_03}}
]
properties {
 attenuationSettings=&Fort_Weapon_Projectile_FirstPerson_Athena_Att
 maxDistance=2097152
 soundClassObject=&Fort_WeaponRanged_Fire_1P_PSM
 volumeMultiplier=1
 attenuationOverrides {distanceAlgorithm=NaturalSound, falloffDistance=5000, attenuationShapeExtents {x=1000}}
 concurrencySet [&Weapon_Gunfire_FirstPerson_Conc]
}
