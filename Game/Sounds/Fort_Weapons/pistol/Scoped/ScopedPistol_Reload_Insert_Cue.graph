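# Sound cue for the insert step of the scoped pistol reload, categorized as weapon foley.
# A single wave player plays one insert sample with no randomization or layering.
# It uses the shared weapon reload attenuation, a slightly reduced volume, and a concurrency override that prevents new instances once its limit is reached.
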SoundCue /Game/Sounds/Fort_Weapons/pistol/Scoped/ScopedPistol_Reload_Insert_Cue
root=WavePlayer_0
nodes [{id=WavePlayer_0, properties {soundWaveAssetPtr=&/ScopedPistol_Reload_Insert_01}}]
properties {
 attenuationSettings=&Fort_Weapon_Reload_Att
 bOverrideConcurrency=true
 maxDistance=2250
 soundClassObject=&Fort_WeaponFoley
 volumeMultiplier=0.802
 concurrencyOverrides {resolutionRule=PreventNew}
}
