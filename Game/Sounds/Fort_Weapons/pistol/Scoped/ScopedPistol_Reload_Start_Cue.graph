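# Sound cue for the start of the scoped pistol reload, played as weapon foley with reload attenuation.
# A top mixer layers the pitch- and volume-modulated reload start sound with a quieter sub-mix of delayed details.
# The sub-mix holds a random pick of two shell sounds and a later-mixed, pitch-varied SixShooter reload layer that starts slightly earlier than the shells.
# Concurrency is overridden to prevent new instances once the limit is reached.
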SoundCue /Game/Sounds/Fort_Weapons/pistol/Scoped/ScopedPistol_Reload_Start_Cue
root=Mixer_0
nodes [
 {id=Delay_0, children [Random_1], properties {delayMax=0.8, delayMin=0.8}}
 {id=Delay_2, children [Modulator_2], properties {delayMax=0.6, delayMin=0.6}}
 {id=Mixer_0, children [Modulator_1,Mixer_3], properties {inputVolume [1,1.2]}}
 {id=Mixer_3, children [Delay_0,Delay_2], properties {inputVolume [0.7,0.2]}}
 {id=Modulator_1, children [WavePlayer_0], properties {pitchMax=1, pitchMin=0.9, volumeMax=1.2, volumeMin=1.1}}
 {id=Modulator_2, children [WavePlayer_9], properties {pitchMax=1, pitchMin=0.9}}
 {id=Random_1, children [WavePlayer_1,WavePlayer_6], properties {weights [1,1]}}
 {id=WavePlayer_0, properties {soundWaveAssetPtr=&/ScopedPistol_Reload_Start_01}}
 {id=WavePlayer_1, properties {soundWaveAssetPtr=&/ScopedPistol_Reload_Shells_01}}
 {id=WavePlayer_6, properties {soundWaveAssetPtr=&/ScopedPistol_Reload_Shells_03}}
 {id=WavePlayer_9, properties {soundWaveAssetPtr=/Game/Sounds/Fort_Weapons/pistol/SixShooter/SixShooter_ReloadStartB_01}}
]
properties {
 attenuationSettings=&Fort_Weapon_Reload_Att
 bOverrideConcurrency=true
 maxDistance=2250
 soundClassObject=&Fort_WeaponFoley
 volumeMultiplier=0.802
 concurrencyOverrides {resolutionRule=PreventNew}
}
